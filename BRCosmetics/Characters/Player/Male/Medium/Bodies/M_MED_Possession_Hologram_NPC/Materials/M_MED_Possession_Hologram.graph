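# Skeletal-mesh character material for the Possession Hologram NPC body, using the Subsurface Profile shading model.
# MF_Holo builds the material attributes from Diffuse and Normals texture parameters.
# Its World Position Offset passes through Fortnite_HideElementsWithVertexColor, then adds the vector toward the object position scaled by HT_CrunchVerts, which pulls vertices toward that point.
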Material /BRCosmetics/Characters/Player/Male/Medium/Bodies/M_MED_Possession_Hologram_NPC/Materials/M_MED_Possession_Hologram
graph {
 properties {
  bUseMaterialAttributes=true
  bUsedWithMeshDeformer=true
  bUsedWithMorphTargets=true
  bUsedWithSkeletalMesh=true
  bUsedWithStaticLighting=true
  numCustomizedUVs=3
  physMaterial=/Game/FortressPhysicalMaterials/Player_Flesh
  pixelDepthOffsetMode=PDOM_Legacy
  shadingModel=MSM_SubsurfaceProfile
  subsurfaceProfile=/Game/Packages/Fortress_SharedMaterials/SubSurfaceProfile/Character_SubsurfaceProfile
  usedShadingModels=Subsurface Profile
 }
 nodes [
  n1 = TextureObjectParameter() {
   parameterName=Normals
   samplerType=SAMPLERTYPE_Normal
   texture=/BRCosmetics/Characters/Player/Male/Medium/Bodies/M_MED_Possession_Hologram_NPC/Textures/T_M_MED_Possession_Hologram_Body_N
  }
  n2 = GetMaterialAttributes(None=MaterialFunctionCall!(BaseColor=TextureObjectParameter() {parameterName=Diffuse}, Normal=@n1) {materialFunction=/Game/UI/Frontend/Lobby/HeroSquad/MF_Holo}.Result) {attributeGetTypes [#1]}
  n3 = Subtract(A=ObjectPositionWS(), B=WorldPosition().XYZ)
  n4 = ScalarParameter() {parameterName=HT_CrunchVerts}
  SetMaterialAttributes!(
   MaterialAttributes=@n2.MaterialAttributes
   World Position Offset=Add(
    A=MaterialFunctionCall!(WPO=@n2.World Position Offset) {materialFunction=/Game/Packages/Fortress_SharedMaterials/Base_Material_Functions/Fortnite_HideElementsWithVertexColor}.Result
    B=Multiply(A=@n3, B=@n4)
   )
  ) {attributeSetTypes [#1]}
 ]
}
